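# Physics asset for the Submariner male two-handed pickaxe, covering the handle and its dynamic propeller.
# Box bodies cover the handle, the propeller and a tiny propeller master bone that links the two.
# The propeller is hinged to the master with only Swing2 free, so it rotates on a single axis.
# The master's joint to the handle leaves both swings and twist free.
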PhysicsAsset /BRCosmetics/Weapons/FORT_Melee/Pickaxe_Submariner_Male/Meshes/Submariner_Male_Axe_Physics
bodies {
 Melee_TwoHanded_Handle [{shapeType=Box, extentX=7.646, extentY=6.3329, extentZ=121.5183, center=<1.2113 -1.5996 23.5836>}]
 dyn_propeller [{shapeType=Box, extentX=13.1549, extentY=5.5794, extentZ=13.1549}]
 dyn_propeller_mstr [{shapeType=Box, extentX=1.01, extentY=1.01, extentZ=1.01}]
}
constraints [
 {
  bone1Name=dyn_propeller
  bone2Name=dyn_propeller_mstr
  swing1LimitDegrees=45
  swing1Motion=Locked
  swing2LimitDegrees=45
  swing2Motion=Free
  twistLimitDegrees=45
  twistMotion=Locked
 }
 {
  bone1Name=dyn_propeller_mstr
  bone2Name=Melee_TwoHanded_Handle
  swing1LimitDegrees=45
  swing1Motion=Free
  swing2LimitDegrees=45
  swing2Motion=Free
  twistLimitDegrees=45
  twistMotion=Free
 }
]
